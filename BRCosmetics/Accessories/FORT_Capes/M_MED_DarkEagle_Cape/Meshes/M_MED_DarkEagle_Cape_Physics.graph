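# Physics asset for the Dark Eagle cape cosmetic, defining collision bodies and joint limits for the cape's dynamic bones.
# The backpack_root body groups capsules and boxes, and anchors a central chain of five thin box bodies.
# That chain forks at dyn_cape05 into left and right chains of six boxes each.
# Swing limits are tighter on the central chain and looser on the branches, with tight twist throughout.
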PhysicsAsset /BRCosmetics/Accessories/FORT_Capes/M_MED_DarkEagle_Cape/Meshes/M_MED_DarkEagle_Cape_Physics
bodies {
 backpack_root [
  {shapeType=Capsule, length=28.9062, radius=16.8125, center=<-0.0005 17.8837 -29.2613>, rotation=<rot 0 0 -11.2499>}
  {shapeType=Capsule, length=14.1875, radius=17.375, center=<-0.0001 18.2946 -55.8843>, rotation=<rot 90 0 0>}
  {shapeType=Capsule, length=26.6875, radius=15, center=<-0.0003 15.1012 -22.0176>, rotation=<rot -90 -0.0149 0.0149>}
  {shapeType=Box, extentX=1, extentY=1, extentZ=1}
  {shapeType=Box, extentX=106.4375, extentY=5.875, extentZ=122.125, center=<0.0003 15.5153 -74.1826>}
 ]
 dyn_cape01 [{shapeType=Box, extentX=8.125, extentY=1, extentZ=31.375, center=<4.2277 0 0>}]
 dyn_cape01_l [{shapeType=Box, extentX=12.5625, extentY=1, extentZ=15.25, center=<6.8854 0 0>}]
 dyn_cape01_r [{shapeType=Box, extentX=12.625, extentY=1, extentZ=15.6875, center=<6.7273 -0.0001 0>}]
 dyn_cape02 [{shapeType=Box, extentX=8, extentY=1, extentZ=31.9375, center=<4.2056 -0.0001 0>}]
 dyn_cape02_l [{shapeType=Box, extentX=12.8125, extentY=1, extentZ=15.4375, center=<7.1259 -0.0002 0>}]
 dyn_cape02_r [{shapeType=Box, extentX=13, extentY=1, extentZ=15.875, center=<7.0843 -0.0002 0>}]
 dyn_cape03 [{shapeType=Box, extentX=7.8125, extentY=1, extentZ=33.9375, center=<4.1885 0 0>}]
 dyn_cape03_l [{shapeType=Box, extentX=12.6875, extentY=1, extentZ=15.5, center=<6.9891 -0.0001 0>}]
 dyn_cape03_r [{shapeType=Box, extentX=12.625, extentY=1, extentZ=15.75, center=<7.0789 -0.0001 0>}]
 dyn_cape04 [{shapeType=Box, extentX=22.25, extentY=1, extentZ=35.1875, center=<10.9811 -0.0001 0>}]
 dyn_cape04_l [{shapeType=Box, extentX=12.8125, extentY=1, extentZ=15.6875, center=<6.9654 -0.0001 0>}]
 dyn_cape04_r [{shapeType=Box, extentX=12.625, extentY=1, extentZ=15.625, center=<7.0114 0.0001 0>}]
 dyn_cape05 [{shapeType=Box, extentX=14.3125, extentY=1, extentZ=35.25, center=<7.457 0.0001 0>}]
 dyn_cape05_l [{shapeType=Box, extentX=12.75, extentY=1, extentZ=15.75, center=<6.9628 0 0>}]
 dyn_cape05_r [{shapeType=Box, extentX=12.1875, extentY=1, extentZ=15.375, center=<6.945 0 0>}]
 dyn_cape06_l [{shapeType=Box, extentX=15.4375, extentY=1, extentZ=15.25, center=<1.8177 0 -8.0923>, rotation=<rot 11.2362 2.8685 0.5592>}]
 dyn_cape06_r [{shapeType=Box, extentX=15.4375, extentY=1, extentZ=13.9375, center=<-1.9814 0.0002 -7.3158>, rotation=<rot -14.0455 -2.8994 0.7042>}]
}
constraints [
 {bone1Name=dyn_cape01, bone2Name=backpack_root, swing1LimitDegrees=20, swing2LimitDegrees=20, twistLimitDegrees=5}
 {bone1Name=dyn_cape02, bone2Name=dyn_cape01, swing1LimitDegrees=20, swing2LimitDegrees=20, twistLimitDegrees=5}
 {bone1Name=dyn_cape03, bone2Name=dyn_cape02, swing1LimitDegrees=10, swing2LimitDegrees=10, twistLimitDegrees=5}
 {bone1Name=dyn_cape04, bone2Name=dyn_cape03, swing1LimitDegrees=10, swing2LimitDegrees=10, twistLimitDegrees=5}
 {bone1Name=dyn_cape05, bone2Name=dyn_cape04, swing1LimitDegrees=15, swing2LimitDegrees=15, twistLimitDegrees=5}
 {bone1Name=dyn_cape01_l, bone2Name=dyn_cape05, swing1LimitDegrees=40, swing2LimitDegrees=25, twistLimitDegrees=5}
 {bone1Name=dyn_cape02_l, bone2Name=dyn_cape01_l, swing1LimitDegrees=40, swing2LimitDegrees=25, twistLimitDegrees=5}
 {bone1Name=dyn_cape03_l, bone2Name=dyn_cape02_l, swing1LimitDegrees=40, swing2LimitDegrees=25, twistLimitDegrees=5}
 {bone1Name=dyn_cape04_l, bone2Name=dyn_cape03_l, swing1LimitDegrees=40, swing2LimitDegrees=25, twistLimitDegrees=5}
 {bone1Name=dyn_cape05_l, bone2Name=dyn_cape04_l, swing1LimitDegrees=40, swing2LimitDegrees=25, twistLimitDegrees=5}
 {bone1Name=dyn_cape06_l, bone2Name=dyn_cape05_l, swing1LimitDegrees=40, swing2LimitDegrees=25, twistLimitDegrees=5}
 {bone1Name=dyn_cape01_r, bone2Name=dyn_cape05, swing1LimitDegrees=40, swing2LimitDegrees=25, twistLimitDegrees=5}
 {bone1Name=dyn_cape02_r, bone2Name=dyn_cape01_r, swing1LimitDegrees=40, swing2LimitDegrees=25, twistLimitDegrees=5}
 {bone1Name=dyn_cape03_r, bone2Name=dyn_cape02_r, swing1LimitDegrees=40, swing2LimitDegrees=25, twistLimitDegrees=5}
 {bone1Name=dyn_cape04_r, bone2Name=dyn_cape03_r, swing1LimitDegrees=40, swing2LimitDegrees=25, twistLimitDegrees=5}
 {bone1Name=dyn_cape05_r, bone2Name=dyn_cape04_r, swing1LimitDegrees=40, swing2LimitDegrees=25, twistLimitDegrees=5}
 {bone1Name=dyn_cape06_r, bone2Name=dyn_cape05_r, swing1LimitDegrees=40, swing2LimitDegrees=25, twistLimitDegrees=5}
]
