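# Animation Blueprint for Fortnite back-bling capes that copies the parent mesh pose and simulates cape motion with RigidBody nodes using the DarkEagle cape physics asset.
# Bool blends select skydiving/parachute physics with variable gravity, moving physics with additive bone offsets, or plain physics otherwise.
# The update event derives that gravity from skydiving force or a downward default, flipping its sign by cape yaw during emotes, and sets velocity/acceleration scales by skydiving state.
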AnimBlueprint /BRCosmetics/Accessories/FORT_Capes/Common/Fortnite_Base_Cape_AnimBlueprint
target_skeleton=/Game/Accessories/FORT_Capes/Common/Fortnite_Base_Cape_Skeleton
graphs [
 AnimGraph {nodes [
   Local Space Ref Pose()
   n1 = RigidBody!(
    ComponentPose=Local To Component(LocalPose=Copy Pose From Mesh() {bUseAttachedParent=true})
    Alpha
    ExternalForce
    ComponentLinearAccScale:Vector="0.5,0.5,0.5"
    ComponentLinearVelScale:Vector="0.4,0.4,0.4"
   ) -> (Pose) {
    bOverrideWorldGravity=true
    overridePhysicsAsset=/BRCosmetics/Accessories/FORT_Capes/M_MED_DarkEagle_Cape/Meshes/M_MED_DarkEagle_Cape_Physics
    simulationSpace=BaseBoneSpace
    alphaBoolBlend {blendOption=Linear}
    baseBoneRef {boneName=backpack_root}
    componentLinearAccScale=<0.5 0.5 0.5>
    componentLinearVelScale=<0.4 0.4 0.4>
    overrideWorldGravity {z=-980}
   }
   n2 = Component To Local(ComponentPose=@n1)
   n3 = RigidBody(
    ComponentPose=Local To Component(LocalPose=Copy Pose From Mesh() {bUseAttachedParent=true})
    ComponentLinearAccScale="0.5,0.5,0.5"
    ComponentLinearVelScale="0.4,0.4,0.4"
   ) {
    bOverrideWorldGravity=true
    overridePhysicsAsset=&M_MED_DarkEagle_Cape_Physics
    simulationSpace=BaseBoneSpace
    alphaBoolBlend {blendOption=Linear}
    baseBoneRef {boneName=backpack_root}
    componentLinearAccScale=<0.5 0.5 0.5>
    componentLinearVelScale=<0.4 0.4 0.4>
    overrideWorldGravity {z=-980}
   }
   n4 = "Transform (Modify) Bone - Bone: dyn_cape01"(ComponentPose=@n3, Alpha=1, Translation="0,-5,0", Rotation="0,0,0", Scale="1,1,1") {
    rotationMode=BMM_Additive
    rotationSpace=BCS_BoneSpace
    translationMode=BMM_Additive
    translationSpace=BCS_BoneSpace
    alphaBoolBlend {blendOption=Linear}
    boneToModify {boneName=dyn_cape01}
    translation {y=-5}
   }
   n5 = "Transform (Modify) Bone - Bone: dyn_cape03"(ComponentPose=@n4, Alpha=1, Translation="0,-2,0", Rotation="0,0,0", Scale="1,1,1") {
    rotationMode=BMM_Additive
    rotationSpace=BCS_BoneSpace
    translationMode=BMM_Additive
    translationSpace=BCS_BoneSpace
    alphaBoolBlend {blendOption=Linear}
    boneToModify {boneName=dyn_cape03}
    translation {y=-2}
   }
   n6 = "Transform (Modify) Bone - Bone: dyn_cape05"(ComponentPose=@n5, Alpha=1, Translation="0,-2,0", Rotation="0,0,0", Scale="1,1,1") {
    rotationMode=BMM_Additive
    rotationSpace=BCS_BoneSpace
    translationMode=BMM_Additive
    translationSpace=BCS_BoneSpace
    alphaBoolBlend {blendOption=Linear}
    boneToModify {boneName=dyn_cape05}
    translation {y=-2}
   }
   n7 = "Transform (Modify) Bone - Bone: dyn_cape06_l"(ComponentPose=@n6, Alpha=1, Translation="0,-7,0", Rotation="0,0,0", Scale="1,1,1") {
    rotationMode=BMM_Additive
    rotationSpace=BCS_BoneSpace
    translationMode=BMM_Additive
    translationSpace=BCS_BoneSpace
    alphaBoolBlend {blendOption=Linear}
    boneToModify {boneName=dyn_cape06_l}
    translation {y=-7}
   }
   n8 = "Transform (Modify) Bone - Bone: dyn_cape06_r"(ComponentPose=@n7, Alpha=1, Translation="0,-7,0", Rotation="0,0,0", Scale="1,1,1") {
    rotationMode=BMM_Additive
    rotationSpace=BCS_BoneSpace
    translationMode=BMM_Additive
    translationSpace=BCS_BoneSpace
    alphaBoolBlend {blendOption=Linear}
    boneToModify {boneName=dyn_cape06_r}
    translation {y=-7}
   }
   n9 = Component To Local(ComponentPose=@n8)
   n10 = OR Boolean!(A=Get bIsMoving2D(), B=Get bIsCrouchSprinting(), C:bool=Get bIsSprinting()) -> (ReturnValue)
   n11 = RigidBody!(
    ComponentPose=Local To Component(LocalPose=Copy Pose From Mesh() {bUseAttachedParent=true})
    Alpha
    OverrideWorldGravity:Vector=Get finalGravity() -> (finalGravity:Vector)
    ExternalForce
    ComponentLinearAccScale=Get compLinearAccScale() -> (compLinearAccScale:Vector)
    ComponentLinearVelScale=Get compLinearVelScale() -> (compLinearVelScale:Vector)
   ) -> (Pose) {
    bOverrideWorldGravity=true
    overridePhysicsAsset=&M_MED_DarkEagle_Cape_Physics
    simulationSpace=BaseBoneSpace
    alphaBoolBlend {blendOption=Linear}
    baseBoneRef {boneName=backpack_root}
    componentAppliedLinearAccClamp=<1500 1500 1500>
    componentLinearAccScale=<0.5 0.5 0.5>
    componentLinearVelScale=<0.4 0.4 0.4>
    overrideWorldGravity {z=-980}
   }
   n12 = Get bIsSkyDiving()
   n13 = Get bIsParachuteOpened()
   Output Pose(Result=Blend Poses by bool(
     BlendPose_0=Component To Local(ComponentPose=@n11)
     BlendPose_1=Blend Poses by bool(BlendPose_0=@n9, BlendPose_1=@n2, BlendTime_0=0.2, BlendTime_1=0.5, bActiveValue=@n10) {blendType=Linear}
     BlendTime_0=0.2
     BlendTime_1=0.5
     bActiveValue=OR Boolean!(A=@n12, B=@n13) -> (ReturnValue)
    ) {blendType=Linear}) {layerGroup=None}
   AnimGraphNode_SubInput_0 = Input Pose()
   Get finalGravity()
   Get compLinearVelScale()
   Get compLinearAccScale()
  ]}
 EventGraph {nodes [
   Event BlueprintUpdateAnimation()
   > n1 = Sequence!(execute) -> (then_0, then_1, then_2:Exec, then_3:Exec, then_4:Exec, then_5:Exec, then_6:Exec)
   n2 = Break Fort Anim Input Skydiving External Force(FortAnimInput_SkydivingExternalForce=Get SkydivingExternalForceInput()) -> (
    bUseSkydivingVectorForce=~
    SkydivingVectorForce=~
    HeadToPelvisDirection=~
    FloatingMultiplier=~
    FloatingAdditive=~
    DivingMultiplier=~
    DivingAdditive=~
    ParachutingMultiplier=~
    ParachutingAdditive=~
   )
   > n3 = Set windVector(windVector:Vector=SelectVector(A=vector * float(A=@n2.SkydivingVectorForce, B=0.1), B="0, 0,-800", bPickA=Get bIsSkyDiving())) -> (Output_Get:Vector)
   n4 = GetOwnerSkeletalMeshComponent()
   n5 = BreakRotator(InRot=BreakTransform(InTransform=GetDeltaTransformFromRefPose(self=@n4, BoneName=dyn_cape01, BaseName=backpack_root)).Rotation)
   n6 = SelectFloat(A=980, B=-980, bPickA="float > float"(A="Absolute (Float)"(A=@n5.Yaw), B=90) {is=title K2Node_CallFunction})
   > Set finalGravity(finalGravity:Vector=SelectVector(A=MakeVector(Y=0, Z=@n6), B=@n3, bPickA=Get bIsPlayingEmote())) -> (Output_Get:Vector)
   n7 = SelectVector(A="0.1,0.1,0.1", B="0.3,0.3,0.35", bPickA=Get bIsSkyDiving())
   Set compLinearVelScale(execute=@n1.then_1, compLinearVelScale:Vector=@n7) -> (Output_Get:Vector)
   > Set compLinearAccScale(compLinearAccScale:Vector=@n7) -> (Output_Get:Vector)
  ]}
]
variables [
 compLinearAccScale {type=Vector}
 compLinearVelScale {type=Vector}
 finalGravity {type=Vector}
 windVector {type=Vector}
]
